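# Translucent, two-sided soft fire material for the Hellfire vehicle wheel cosmetic, tinted by CustomColor.
# Three panned samples of a shared flow texture, timed by FlameSpeed, distort vertical UVs into flame shapes.
# The flame mask drives a color gradient into BaseColor, with EmissiveColor left unconnected despite the EmissiveStrength parameter.
# Opacity combines that mask with UV edge, world-space view-angle and depth fades, shaped by vertex color red and alpha.
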Material /VehicleCosmetics/Wheels/Hellfire/Materials/M_Hellfire_FireSoft
graph {
 properties {
  bAutomaticallySetUsageInEditor=false
  bUseMaterialAttributes=true
  bUsedWithMeshDeformer=true
  bUsedWithSkeletalMesh=true
  bUsedWithStaticLighting=true
  blendMode=BLEND_Translucent
  twoSided=true
  usedShadingModels=Default Lit
 }
 nodes [
  n1 = VectorParameter() {parameterName=CustomColor, defaultValue=<rgba 1 0.35 0.1 1>}
  n2 = ScalarParameter() {defaultValue=3, parameterName=EmissiveStrength}
  n3 = Multiply(A=Multiply(A=@n1.RGB, B=@n1.RGB), B=Multiply(B=@n2) {constA=0.5}) {constA=0.5}
  n4 = Multiply(A=@n3, B=@n3) {constA=0.5}
  n5 = Multiply(A=@n1.RGB, B=@n2) {constA=0.5}
  n6 = TextureCoordinate() {vTiling=0.75}
  n7 = ComponentMask(None=@n6) {g=true}
  n8 = ConstantBiasScale(None=@n7) {bias=0.1, scale=1.4}
  n9 = TextureCoordinate()
  n10 = AppendVector(
   A=Multiply(A=ScalarParameter() {defaultValue=0.6, parameterName=XScale}, B=ComponentMask(None=@n9) {r=true})
   B=Multiply(A=ScalarParameter() {defaultValue=0.4, parameterName=YScale}, B=ComponentMask(None=@n9) {g=true})
  )
  n11 = ComponentMask(None=Multiply(A=Time(), B=ScalarParameter() {defaultValue=2, parameterName=FlameSpeed})) {r=true}
  n12 = Multiply(A=@n10) {constB=1.5}
  n13 = Add(
   A=TextureSample(UVs=Panner(Coordinate=@n10, Time=@n11) {speedY=0.3}) {samplerType=SAMPLERTYPE_Normal, texture=/VehicleCosmetics/Wheels/_SharedTextures/FX/Flow_03_nomips}.RGBA
   B=TextureSample(UVs=Panner(Coordinate=@n12, Time=@n11) {speedX=-0.01, speedY=0.2}) {samplerType=SAMPLERTYPE_Normal, texture=&Flow_03_nomips}.RGBA
  )
  n14 = TextureSample(UVs=Panner(Coordinate=Multiply(A=@n10) {constB=0.5}, Time=@n11) {speedY=0.1}) {samplerType=SAMPLERTYPE_Normal, texture=&Flow_03_nomips}
  n15 = Add(A=ComponentMask(None=Add(A=@n13, B=@n14.RGBA)) {g=true, r=true}, B=Clamp(None=@n8))
  n16 = ComponentMask(None=Add(A=Add(A=@n6, B=@n15)) {constB=-0.5}) {g=true}
  n17 = OneMinus(None=@n16)
  n18 = OneMinus(None=Clamp(None=Power(Base=@n17) {constExponent=20}))
  n19 = Clamp(None=ConstantBiasScale(None=@n7) {bias=0, scale=8})
  n20 = Multiply(A=@n19, B=@n19)
  n21 = OneMinus(None=CameraVectorWS())
  n22 = DotProduct(A=@n21, B=Constant3Vector() {constant=<rgba 0 0 -1 0>})
  n23 = DotProduct(A=@n21, B=Constant3Vector() {constant=<rgba 0 0 1 0>})
  n24 = Add(A=Clamp(None=ConstantBiasScale(None=@n22) {bias=-0.2, scale=3}), B=Clamp(None=ConstantBiasScale(None=@n23) {bias=-0.2, scale=3}))
  n25 = Multiply(A=@n16, B=Multiply(A=Multiply(A=@n18, B=@n20), B=Clamp(None=@n24)))
  n26 = ScalarParameter() {defaultValue=0.8, parameterName=OpacityMultiply}
  n27 = ScalarParameter() {defaultValue=0.3, parameterName=OpacityPower}
  n28 = OneMinus(None=VertexColor().R)
  n29 = Clamp(None=Power(Base=Multiply(A=@n25, B=@n26), Exp=Multiply(A=@n27, B=@n28)))
  MakeMaterialAttributes!(
   BaseColor=Multiply(A=LinearInterpolate(A=@n4, B=@n5, Alpha=@n8), B=@n16)
   Metallic
   Specular
   Roughness
   Anisotropy
   EmissiveColor
   Opacity=Multiply(A=VertexColor().A, B=DepthFade(Opacity=@n29) {fadeDistanceDefault=-3})
   OpacityMask
   Normal
   Tangent
   WorldPositionOffset
   SubsurfaceColor
   ClearCoat
   ClearCoatRoughness
   AmbientOcclusion
   Refraction
   CustomizedUVs_0
   CustomizedUVs_1
   CustomizedUVs_2
   CustomizedUVs_3
   CustomizedUVs_4
   CustomizedUVs_5
   CustomizedUVs_6
   CustomizedUVs_7
   PixelDepthOffset
   ShadingModel
   Displacement
  )
 ]
}
