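# Lit material for the Hellfire cosmetic wheel, drawing animated, tintable lava runes and glowing rock on skeletal meshes.
# Panned rock and water textures plus a sine pulse form the CustomColor-tinted lava, added to both base color and emissive.
# Vertex alpha and green, combined with a transformed-Z-axis mask, blend between pulsing lava and a simpler rock-pattern glow.
# Vertex red adds a separately tinted rock glow from a panning flow texture to emissive.
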Material /VehicleCosmetics/Wheels/Hellfire/Materials/M_Hellfire_LavaRune
graph {
 properties {
  bAutomaticallySetUsageInEditor=false
  bUseMaterialAttributes=true
  bUsedWithMeshDeformer=true
  bUsedWithSkeletalMesh=true
  usedShadingModels=Default Lit
 }
 nodes [
  n1 = Time()
  n2 = Panner(Time=Multiply(A=ScalarParameter() {defaultValue=0.2, parameterName=pannerspeed}, B=@n1)) {speedY=0.2}
  n3 = TextureSample(UVs=@n2) {texture=/VehicleCosmetics/Wheels/_SharedTextures/FX/Detail_Rocky_RGB}
  n4 = OneMinus(None=Power(Base=@n3.G))
  n5 = Subtract(A=Sine(None=@n1) {period=4}, B=@n4)
  n6 = Multiply(A=@n3.G, B=@n3.G)
  n7 = TextureSample(UVs=Panner() {speedX=-0.1}) {texture=/VehicleCosmetics/Wheels/_SharedTextures/FX/Water_01_Pack}
  n8 = VectorParameter() {parameterName=CustomColor, defaultValue=<rgba 0.8784 0.3144 0 1>}
  n9 = Multiply(A=Add(A=Clamp(None=@n5), B=Multiply(A=@n6, B=@n7.G)), B=Multiply(A=@n8.RGB) {constB=12}) {constB=12}
  n10 = Multiply(A=@n8.RGB, B=ScalarParameter() {defaultValue=6, parameterName=EmissiveIntensityMultiply}) {constB=12}
  n11 = Power(Base=VertexColor().G) {constExponent=4}
  n12 = Constant3Vector() {constant=<rgba 0 0 1 0>}
  n13 = Transform(None=@n12)
  n14 = Multiply(
   A=ScalarParameter() {defaultValue=0.2, parameterName=EmissiveMaskHardness}
   B=Add(A=ScalarParameter() {defaultValue=4.5, parameterName=EmissiveMaskPosition}, B=DotProduct(A=@n12, B=@n13))
  )
  n15 = OneMinus(None=ConstantBiasScale(None=ComponentMask(None=@n14) {b=true}) {bias=0, scale=1.05})
  n16 = OneMinus(None=Multiply(A=Clamp(None=@n11), B=Clamp(None=@n15)))
  n17 = ReflectionVectorWS()
  n18 = LinearInterpolate(
   A=@n9
   B=LinearInterpolate(A=Multiply(A=@n4, B=@n10), B=@n9, Alpha=Clamp(None=@n16) {maxDefault=4})
   Alpha=Add(A=VertexColor().A, B=Multiply(A=@n17) {constB=0})
  )
  n19 = VertexColor()
  n20 = ScalarParameter() {defaultValue=7, parameterName=RockGlowMultiply}
  n21 = TextureSample(UVs=@n2) {samplerType=SAMPLERTYPE_Normal, texture=/VehicleCosmetics/Wheels/_SharedTextures/FX/Flow_03_nomips}
  n22 = Multiply(
   A=Multiply(
    A=Multiply(A=@n8.RGB, B=@n20) {constB=12}
    B=VectorParameter() {parameterName=RockGlowMultiplycolor, defaultValue=<rgba 1 0.6022 0.3125 1>}.RGB
   ) {constB=12}
   B=Clamp(None=Add(A=@n21.B, B=@n21.G))
  ) {constB=12}
  MakeMaterialAttributes!(
   BaseColor=Add(A=VectorParameter() {parameterName=Rocks, defaultValue=<rgba 0.0276 0.0251 0.023 1>}.RGB, B=@n18)
   Metallic
   Specular=ScalarParameter() {defaultValue=128, parameterName=Scpecular}
   Roughness
   Anisotropy
   EmissiveColor=Add(A=@n18, B=Multiply(A=@n19.R, B=@n22) {constB=12})
   Opacity
   OpacityMask
   Normal
   Tangent
   WorldPositionOffset
   SubsurfaceColor
   ClearCoat
   ClearCoatRoughness
   AmbientOcclusion
   Refraction
   CustomizedUVs_0
   CustomizedUVs_1
   CustomizedUVs_2
   CustomizedUVs_3
   CustomizedUVs_4
   CustomizedUVs_5
   CustomizedUVs_6
   CustomizedUVs_7
   PixelDepthOffset
   ShadingModel
   Displacement
  )
  Power(Base=OneMinus(None=@n3.G)) {constExponent=6}
 ]
}
